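# Sound cue for the LimeRock Season 37 event music stinger tied to the monster appearance.
# A single wave player feeds the output with the Event 04 TentacleMonster stinger wave, with no mixing, randomization or modulation nodes.
# It uses the LimeRock music attenuation asset with a very large max distance and routes through the Fortnite events music sound class.
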SoundCue /LimeRock/Assets/Audio/Music/MX_S37_LimeRock_Event_04_MonsterAppear_STNG
root=WavePlayer_0
nodes [{id=WavePlayer_0, properties {soundWaveAssetPtr=/LimeRock/Assets/Audio/Music/Wav/MX_S37_LimeRock_Event_04_TentacleMonster_STNG}}]
properties {
 attenuationSettings=/LimeRock/Assets/Audio/Settings/ATT_LimeRock_Music
 maxDistance=120000
 soundClassObject=/Game/Sounds/FortSoundClassesAndModes/Fort_Events_Music
}
